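# Physics asset for the M_MED_Seaweed male player body cosmetic, covering only its left and right "dyn_tube" bone chains, apparently for secondary motion.
# On each side, a tiny MSTR capsule acts as the anchor for a tube_01 capsule, joined with limited swing and a tighter twist.
# Each tube_02 box has swing and twist locked to its tube_01, so each tube pivots as one rigid piece at its root.
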PhysicsAsset /BRCosmetics/Characters/Player/Male/Medium/Bodies/M_MED_Seaweed/Meshes/M_MED_Seaweed_Physics
bodies {
 dyn_tube_01_l [{shapeType=Capsule, length=6.5, radius=3, center=<4.1528 0 0>, rotation=<rot 0 -89.9999 89.9995>}]
 dyn_tube_01_l_MSTR [{shapeType=Capsule, length=0.01, radius=0.01}]
 dyn_tube_01_r [{shapeType=Capsule, length=6.5, radius=3, center=<4.1528 0 0>, rotation=<rot 0 -89.9999 89.9995>}]
 dyn_tube_01_r_MSTR [{shapeType=Capsule, length=0.01, radius=0.01}]
 dyn_tube_02_l [{shapeType=Box, extentX=4.6544, extentY=5, extentZ=5}]
 dyn_tube_02_r [{shapeType=Box, extentX=4.6544, extentY=5, extentZ=5}]
}
constraints [
 {bone1Name=dyn_tube_01_r, bone2Name=dyn_tube_01_r_MSTR, swing1LimitDegrees=25, swing2LimitDegrees=25, twistLimitDegrees=5}
 {bone1Name=dyn_tube_01_l, bone2Name=dyn_tube_01_l_MSTR, swing1LimitDegrees=25, swing2LimitDegrees=25, twistLimitDegrees=5}
 {
  bone1Name=dyn_tube_02_l
  bone2Name=dyn_tube_01_l
  swing1LimitDegrees=45
  swing1Motion=Locked
  swing2LimitDegrees=45
  swing2Motion=Locked
  twistLimitDegrees=45
  twistMotion=Locked
 }
 {
  bone1Name=dyn_tube_02_r
  bone2Name=dyn_tube_01_r
  swing1LimitDegrees=45
  swing1Motion=Locked
  swing2LimitDegrees=45
  swing2Motion=Locked
  twistLimitDegrees=45
  twistMotion=Locked
 }
]
