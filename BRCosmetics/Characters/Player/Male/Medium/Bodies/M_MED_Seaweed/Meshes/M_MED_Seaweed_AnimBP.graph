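# Post-process style AnimBlueprint for the M_MED_Seaweed player body, adding secondary motion and glove correction to the pose copied from the parent mesh.
# Weighted constraints blend tube bones between pelvis, spine and neighboring tube bones, then a pelvis-relative RigidBody simulation is applied at partial alpha.
# Each update maps upper-arm to lower-arm-twist distance into per-side alphas, so additive glove bone offsets fade in as each arm bends.
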AnimBlueprint /BRCosmetics/Characters/Player/Male/Medium/Bodies/M_MED_Seaweed/Meshes/M_MED_Seaweed_AnimBP
target_skeleton=&/M_MED_Seaweed_Skeleton
graphs [
 AnimGraph {nodes [
   n1 = "Constraint - dyn_tube_bk_01_r (PPPPP)"!(
    ComponentPose=Local To Component(LocalPose=Copy Pose From Mesh() {bUseAttachedParent=true})
    Alpha=1
    ConstraintWeights_0=0
    ConstraintWeights_1:float=0.1589
    ConstraintWeights_2:float=0.4171
    ConstraintWeights_3:float=0.3982
    ConstraintWeights_4:float=0.0258
   ) -> (Pose) {
    alphaBoolBlend {blendOption=Linear}
    boneToModify {boneName=dyn_tube_bk_01_r}
    constraintSetup [
     {targetBone {boneName=spine_01}}
     {targetBone {boneName=spine_02}}
     {targetBone {boneName=spine_03}}
     {targetBone {boneName=spine_04}}
     {targetBone {boneName=spine_05}}
    ]
    constraintWeights [1,1,1,1,1]
   }
   n2 = "Constraint - dyn_tube_01_r (PPPPP)"!(
    ComponentPose=@n1
    Alpha=1
    ConstraintWeights_0=0.0464
    ConstraintWeights_1:float=0.3382
    ConstraintWeights_2:float=0.3128
    ConstraintWeights_3:float=0.1276
    ConstraintWeights_4:float=0.1751
   ) -> (Pose) {
    alphaBoolBlend {blendOption=Linear}
    boneToModify {boneName=dyn_tube_01_r}
    constraintSetup [
     {targetBone {boneName=pelvis}}
     {targetBone {boneName=spine_02}}
     {targetBone {boneName=spine_03}}
     {targetBone {boneName=spine_04}}
     {targetBone {boneName=spine_01}}
    ]
    constraintWeights [1,1,1,1,1]
   }
   n3 = "Constraint - dyn_tube_fr_01_r (PPPPP)"!(
    ComponentPose=@n2
    Alpha=1
    ConstraintWeights_0=0.3136
    ConstraintWeights_1:float=0.191
    ConstraintWeights_2:float=0.0881
    ConstraintWeights_3:float=0.005
    ConstraintWeights_4:float=0.4023
   ) -> (Pose) {
    alphaBoolBlend {blendOption=Linear}
    boneToModify {boneName=dyn_tube_fr_01_r}
    constraintSetup [
     {targetBone {boneName=pelvis}}
     {targetBone {boneName=spine_02}}
     {targetBone {boneName=spine_03}}
     {targetBone {boneName=spine_04}}
     {targetBone {boneName=spine_01}}
    ]
    constraintWeights [1,1,1,1,1]
   }
   n4 = "Constraint - dyn_tube_01_l (PPPPP)"!(
    ComponentPose=@n3
    Alpha=1
    ConstraintWeights_0=0.0464
    ConstraintWeights_1:float=0.3382
    ConstraintWeights_2:float=0.3128
    ConstraintWeights_3:float=0.1276
    ConstraintWeights_4:float=0.1751
   ) -> (Pose) {
    alphaBoolBlend {blendOption=Linear}
    boneToModify {boneName=dyn_tube_01_l}
    constraintSetup [
     {targetBone {boneName=pelvis}}
     {targetBone {boneName=spine_02}}
     {targetBone {boneName=spine_03}}
     {targetBone {boneName=spine_04}}
     {targetBone {boneName=spine_01}}
    ]
    constraintWeights [1,1,1,1,1]
   }
   n5 = "Constraint - dyn_tube_bk_01_l (PPPPP)"!(
    ComponentPose=@n4
    Alpha=1
    ConstraintWeights_0=0
    ConstraintWeights_1:float=0.1589
    ConstraintWeights_2:float=0.4171
    ConstraintWeights_3:float=0.3982
    ConstraintWeights_4:float=0.0258
   ) -> (Pose) {
    alphaBoolBlend {blendOption=Linear}
    boneToModify {boneName=dyn_tube_bk_01_l}
    constraintSetup [
     {targetBone {boneName=spine_01}}
     {targetBone {boneName=spine_02}}
     {targetBone {boneName=spine_03}}
     {targetBone {boneName=spine_04}}
     {targetBone {boneName=spine_05}}
    ]
    constraintWeights [1,1,1,1,1]
   }
   n6 = "Constraint - dyn_tube_fr_01_l (PPPPP)"!(
    ComponentPose=@n5
    Alpha=1
    ConstraintWeights_0=0.3136
    ConstraintWeights_1:float=0.191
    ConstraintWeights_2:float=0.0881
    ConstraintWeights_3:float=0.005
    ConstraintWeights_4:float=0.4023
   ) -> (Pose) {
    alphaBoolBlend {blendOption=Linear}
    boneToModify {boneName=dyn_tube_fr_01_l}
    constraintSetup [
     {targetBone {boneName=pelvis}}
     {targetBone {boneName=spine_02}}
     {targetBone {boneName=spine_03}}
     {targetBone {boneName=spine_04}}
     {targetBone {boneName=spine_01}}
    ]
    constraintWeights [1,1,1,1,1]
   }
   n7 = "Constraint - dyn_tube_02_r (PP)"!(ComponentPose=@n6, Alpha=1, ConstraintWeights_0=0.5, ConstraintWeights_1:float=0.5) -> (Pose) {
    alphaBoolBlend {blendOption=Linear}
    boneToModify {boneName=dyn_tube_02_r}
    constraintSetup [{targetBone {boneName=dyn_tube_bk_01_r}}, {targetBone {boneName=dyn_tube_fr_01_r}}]
    constraintWeights [1,1]
   }
   n8 = RigidBody!(ComponentPose=@n7, Alpha=0.25, OverrideWorldGravity:Vector="0,0,-975", ExternalForce) -> (Pose) {
    bClampLinearTranslationLimitToRefPose=true
    bOverrideWorldGravity=true
    overridePhysicsAsset=&/M_MED_Seaweed_Physics
    simulationSpace=BaseBoneSpace
    alphaBoolBlend {blendOption=Linear}
    baseBoneRef {boneName=pelvis}
    componentAppliedLinearAccClamp=<3000 3000 3000>
    componentLinearAccScale=<0.3 0.3 0.3>
    componentLinearVelScale=<0.3 0.3 0.3>
    overrideWorldGravity {z=-980}
   }
   n9 = "Transform (Modify) Bone - Bone: glove_01_l"(
    ComponentPose=@n8
    Alpha=Get LarmDistanceAlpha() -> (LarmDistanceAlpha:double)
    Translation="0.000,0.677,-2.942"
    Rotation="0.000,0.000,39.999"
    Scale="1,1,1"
   ) {
    rotationMode=BMM_Additive
    rotationSpace=BCS_BoneSpace
    scaleMode=BMM_Replace
    scaleSpace=BCS_BoneSpace
    translationMode=BMM_Additive
    translationSpace=BCS_BoneSpace
    alphaBoolBlend {blendOption=Linear}
    boneToModify {boneName=glove_01_l}
    rotation {roll=39.9987}
    translation=<0 0.6767 -2.9418>
   }
   n10 = "Transform (Modify) Bone - Bone: glove_01_r"(
    ComponentPose=@n9
    Alpha=Get RarmDistanceAlpha() -> (RarmDistanceAlpha:double)
    Translation="0.000,0.677,-2.942"
    Rotation="0.000,0.000,-39.999"
    Scale="1,1,1"
   ) {
    rotationMode=BMM_Additive
    rotationSpace=BCS_BoneSpace
    scaleMode=BMM_Replace
    scaleSpace=BCS_BoneSpace
    translationMode=BMM_Additive
    translationSpace=BCS_BoneSpace
    alphaBoolBlend {blendOption=Linear}
    boneToModify {boneName=glove_01_r}
    rotation {roll=-39.999}
    translation {y=0.677, z=-2.942}
   }
   Output Pose(Result=Component To Local(ComponentPose=@n10)) {layerGroup=None}
  ]}
 EventGraph {nodes [
   Event BlueprintUpdateAnimation()
   n1 = GetPartSkeletalMeshComponent()
   > n2 = TransformFromBoneSpace!(
    execute
    self:Skinned Mesh Component Object Reference=@n1
    BoneName:Name=upperarm_l
    InPosition:Vector="0, 0, 0"
    InRotation:Rotator="0, 0, 0"
   ) -> (then, OutPosition:Vector, OutRotation:Rotator)
   > n3 = TransformFromBoneSpace(self=@n1, BoneName=lowerarm_twist_02_l, InPosition="0, 0, 0", InRotation="0, 0, 0")
   > Set LarmDistanceAlpha(LarmDistanceAlpha:double=MapRangeClamped(Value="Distance (Vector)"!(V1:Vector=@n2, V2:Vector=@n3) -> (ReturnValue:double), InRangeA=27, InRangeB=34, OutRangeA=1, OutRangeB=0)) -> (Output_Get:double)
   > n4 = TransformFromBoneSpace(self=@n1, BoneName=upperarm_r, InPosition="0, 0, 0", InRotation="0, 0, 0")
   > n5 = TransformFromBoneSpace(self=@n1, BoneName=lowerarm_twist_02_r, InPosition="0, 0, 0", InRotation="0, 0, 0")
   > Set RarmDistanceAlpha(RarmDistanceAlpha:double=MapRangeClamped(Value="Distance (Vector)"(V1=@n4, V2=@n5), InRangeA=27, InRangeB=34, OutRangeA=1, OutRangeB=0)) -> (Output_Get:double)
  ]}
]
variables [
 LarmDistanceAlpha {type=number}
 RarmDistanceAlpha {type=number}
]
